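# Translucent, two-sided unlit Niagara material for the side aura of the Dual Paradox emote.
# Particle color drives emissive, optionally tinted by a curve-atlas gradient along U.
# Opacity multiplies a scrolling, noise-warped aura texture by distorted noise, UV edge masks, particle alpha and view-angle fade; dynamic parameters and vertex red vary offset and speed.
# Depth fade applies only on default and deferred paths; forward and mobile skip it.
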Material /BRCosmetics/Effects/Fort_Effects/Emotes/DualParadox/M_Dual_Paradox_Aura_sides
graph {
 properties {
  bUsedWithNiagaraMeshParticles=true
  bUsedWithNiagaraSprites=true
  blendMode=BLEND_Translucent
  pixelDepthOffsetMode=PDOM_Legacy
  shadingModel=MSM_Unlit
  twoSided=true
  usedShadingModels=Unlit
 }
 nodes [
  n1 = ParticleColor()
  n2 = CurveAtlasRowParameter(CurveTime=ComponentMask(None=TextureCoordinate()) {r=true}) {
   atlas=/Game/Effects/Fort_Effects/Emotes/DualParadox/Curve_Atlas_DBZ_Emote
   curve=/Game/Effects/Fort_Effects/Emotes/DualParadox/DualParadox_color02_Curve
   defaultValue=4
  }
  MaterialFunctionCall(Emissive=LinearInterpolate(A=@n1.RGB, B=Multiply(A=@n2, B=@n1.RGB), Alpha=ScalarParameter() {parameterName=Use gradient color})) {materialFunction=&MF_FX_EmissiveAdjustments}
  n3 = Panner(Coordinate=TextureCoordinate() {vTiling=1.5}) {speedY=0.5}
  n4 = TextureCoordinate()
  n5 = Multiply(
   A=ComponentMask(None=TextureSample(UVs=@n3) {samplerType=SAMPLERTYPE_Normal, texture=&T_Weap_Distort_N_02}.RGB) {g=true, r=true}
   B=OneMinus(None=ComponentMask(None=@n4) {g=true})
  )
  n6 = Add(A=TextureCoordinate() {uTiling=0.7, vTiling=0.8}, B=Constant2Vector() {g=0.5})
  n7 = Panner(
   Coordinate=Add(
    A=Multiply(A=@n5) {constB=0.3}
    B=MaterialFunctionCall(XY=@n6) {materialFunction=&Swizzle}.YX
   )
   Time=DynamicParameter() {paramNames [Random, Time Speed, Param3, Param4]}.Time Speed
  ) {speedX=1}
  n8 = TextureSample() {samplerType=SAMPLERTYPE_LinearGrayscale, texture=/Engine/Functions/Engine_MaterialFunctions02/ExampleContent/Textures/LowResBlurredNoise}
  n9 = Power(
   Base=LinearInterpolate(
    A=TextureSample(UVs=@n7) {samplerType=SAMPLERTYPE_LinearColor, texture=&T_NoisePack_LinearColor}.G
    B=@n8.R
    Alpha=Constant() {r=0.5}
   ) {constAlpha=0.2}
   Exp=ScalarParameter() {defaultValue=2, parameterName=Color noise power}
  )
  n10 = TextureCoordinate()
  n11 = ComponentMask(None=VectorParameter() {parameterName=Default UV scale, defaultValue=<rgba 0.5 1.3 0 0>}.RGB) {g=true, r=true}
  n12 = Constant()
  n13 = DynamicParameter() {paramNames [Random, Time Speed, Param3, Param4]}
  n14 = Add(A=Add(A=Multiply(A=@n10, B=@n11), B=AppendVector(A=@n12, B=@n13.Random)), B=AppendVector(A=Constant(), B=VertexColor().R)) {constB=0}
  n15 = ComponentMask(None=VectorParameter() {parameterName=UV_Offset, defaultValue=<rgba 0.3 0 0 1>}.RGB) {g=true, r=true}
  n16 = TextureCoordinate()
  n17 = ComponentMask(None=VectorParameter() {parameterName=Noise UV Scale, defaultValue=<rgba 0.5 1 0 0>}.RGB) {g=true, r=true}
  n18 = VectorParameter() {parameterName=Noise Speed, defaultValue=<rgba -0.2 0.2 0 0>}
  n19 = TextureSampleParameter2D(UVs=Panner(Coordinate=Multiply(A=@n16, B=@n17), Speed=ComponentMask(None=@n18.RGB) {g=true, r=true}) {speedX=-0.2, speedY=0.2}) {parameterName=Noise Tex, samplerType=SAMPLERTYPE_LinearColor, texture=/Game/Effects/Fort_Effects/BluGlow/Violentium/Violentium_Mask}
  n20 = ScalarParameter() {defaultValue=0.05, parameterName=Noise intensity}
  n21 = Panner(
   Coordinate=Add(A=Add(A=@n14, B=@n15), B=Multiply(A=@n19.R, B=@n20) {constB=0.05}) {constB=0}
   Time=DynamicParameter() {paramNames [Random, Time Speed, Param3, Param4]}.Time Speed
  ) {speedY=2.5}
  n22 = TextureCoordinate()
  n23 = Multiply(A=Power(Base=ComponentMask(None=@n22) {r=true}, Exp=ScalarParameter() {defaultValue=2, parameterName=Mask power})) {constB=5}
  n24 = Sine(None=ComponentMask(None=TextureCoordinate()) {g=true}) {period=2}
  n25 = Multiply(
   A=StaticComponentMaskParameter(None=TextureSampleParameter2D(UVs=@n21) {parameterName=Base_Texture_Mask, texture=/BRCosmetics/Effects/Fort_Effects/Emotes/Energize_Max/LV3/T_Energize_Max_Aura_LV3_Side}.RGB) {defaultR=true, parameterName=Custom Mask}
   B=Multiply(A=Saturate(None=@n23), B=Power(Base=@n24))
  )
  n26 = Multiply(
   A=Multiply(A=@n1.A, B=Multiply(A=@n9, B=@n25) {constB=2})
   B=MaterialFunctionCall("FadeSensitivity (optional)"=Constant() {r=2}) {materialFunction=&FadeBasedOnViewAngle}.Result
  )
  n27 = DepthFade(Opacity=@n26) {fadeDistanceDefault=30}
  ShadingPathSwitch(Default=@n27, Deferred=@n27, Forward=@n26, Mobile=@n26)
  TextureCoordinate() {uTiling=0.5, vTiling=1.3}
  TextureCoordinate() {uTiling=0.5}
 ]
}
