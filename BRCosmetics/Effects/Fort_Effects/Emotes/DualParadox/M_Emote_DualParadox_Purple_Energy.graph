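# Translucent, two-sided unlit Niagara particle material for the purple energy effect of the DualParadox emote.
# Color blends two tints through noise warped by a scrolling normal map, then adds a curve-atlas vertical gradient and tinted noise before particle color and emissive adjustment.
# Opacity combines scrolling noises with U-only sphere masks and a sine along V to soften the edges, scaled by particle alpha.
# A dynamic Time Speed parameter drives every panner.
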Material /BRCosmetics/Effects/Fort_Effects/Emotes/DualParadox/M_Emote_DualParadox_Purple_Energy
graph {
 properties {
  bUsedWithNiagaraMeshParticles=true
  bUsedWithNiagaraSprites=true
  blendMode=BLEND_Translucent
  pixelDepthOffsetMode=PDOM_Legacy
  shadingModel=MSM_Unlit
  twoSided=true
  usedShadingModels=Unlit
 }
 nodes [
  n1 = TextureCoordinate() {uTiling=4, vTiling=2}
  n2 = DynamicParameter() {defaultValue=<rgba 0 1 1 1>, paramNames [Random, Time Speed, Param3, Param4]}
  n3 = TextureSample(UVs=Panner(Coordinate=TextureCoordinate() {uTiling=2, vTiling=2}, Time=@n2.Time Speed) {speedY=0.5}) {samplerType=SAMPLERTYPE_Normal, texture=/Game/Characters/Player/Female/Medium/Bodies/F_MED_Bistro_Astronaut/FX/T_GaussianNoise_Normal}
  n4 = Multiply(A=@n3.R) {constB=0.2}
  n5 = TextureSample(UVs=Panner(
    Coordinate=Add(A=@n1, B=@n4)
    Time=DynamicParameter() {defaultValue=<rgba 0 1 1 1>, paramNames [Random, Time Speed, Param3, Param4]}.Time Speed
   ) {speedY=2.5}) {samplerType=SAMPLERTYPE_Grayscale, texture=&LowResBlurredNoise}
  n6 = CurveAtlasRowParameter(CurveTime=ComponentMask(None=TextureCoordinate()) {g=true}) {
   atlas=/Game/Effects/Fort_Effects/Emotes/DualParadox/Curve_Atlas_DBZ_Emote
   curve=/Game/Effects/Fort_Effects/Emotes/DualParadox/DualParadox_color03_Curve
  }
  n7 = Add(
   A=LinearInterpolate(
    A=VectorParameter() {defaultValue=<rgba 0.0027 0.0003 0.0156 1>}.RGB
    B=VectorParameter() {parameterName=Param_1, defaultValue=<rgba 0.2315 0 1.5 1>}.RGB
    Alpha=Power(Base=@n5.R)
   )
   B=Multiply(A=@n6) {constB=0.7}
  ) {constB=0}
  n8 = Multiply(A=@n3.R) {constB=0.1}
  n9 = Panner(Coordinate=TextureCoordinate() {uTiling=4, vTiling=2}, Time=@n2.Time Speed) {speedY=1}
  n10 = Multiply(A=TextureSample(UVs=Add(A=@n8, B=@n9)) {texture=/Game/Effects/Fort_Effects/Textures/Tile/T_Noise_03}.RGB) {constB=1.3}
  n11 = Constant() {r=-1}
  n12 = Add(A=Multiply(A=Desaturation(None=@n10, Fraction=@n11), B=Constant3Vector() {constant=<rgba 0.2733 0.1271 0.8854 1>})) {constB=0}
  n13 = ParticleColor()
  MaterialFunctionCall(Emissive=Multiply(A=Add(A=@n7, B=@n12) {constB=0}, B=@n13.RGB)) {materialFunction=&MF_FX_EmissiveAdjustments}
  n14 = VectorParameter() {parameterName=UV offset, defaultValue=<rgba 0 0 0 1>}
  n15 = Panner(
   Coordinate=Add(A=TextureCoordinate() {uTiling=2, vTiling=4}, B=ComponentMask(None=@n14.RGB) {g=true, r=true})
   Time=DynamicParameter() {defaultValue=<rgba 0 1 1 1>, paramNames [Random, Time Speed, Param3, Param4]}.Time Speed
  ) {speedY=2}
  n16 = SphereMask(A=TextureCoordinate() {vTiling=0}, B=Constant2Vector() {r=0.5}) {attenuationRadius=0.5, hardnessPercent=0}
  n17 = Multiply(A=Add(
    A=TextureSample(UVs=@n15) {samplerType=SAMPLERTYPE_Normal, texture=&T_Weap_Distort_N_02}.R
    B=Power(Base=@n16) {constExponent=1}
   )) {constB=2}
  n18 = Panner(
   Coordinate=Add(A=TextureCoordinate() {uTiling=4, vTiling=2}) {constB=0}
   Time=DynamicParameter() {defaultValue=<rgba 0 1 1 1>, paramNames [Random, Time Speed, Param3, Param4]}.Time Speed
  ) {speedY=1}
  n19 = OneMinus(None=SphereMask(A=TextureCoordinate() {vTiling=0}, B=Constant2Vector() {r=0.5}) {attenuationRadius=0.4, hardnessPercent=0})
  n20 = Power(
   Base=Power(Base=TextureSample(UVs=@n18) {samplerType=SAMPLERTYPE_Grayscale, texture=&LowResBlurredNoise}.R)
   Exp=LinearInterpolate(Alpha=Power(Base=@n19) {constExponent=1}) {constA=0.1, constB=1.5}
  )
  n21 = TextureCoordinate() {vTiling=0}
  n22 = Constant2Vector() {r=0.5}
  n23 = Multiply(A=Saturate(None=Multiply(A=@n17, B=@n20)), B=Power(Base=SphereMask(A=@n21, B=@n22) {attenuationRadius=0.5, hardnessPercent=0}))
  n24 = TextureCoordinate()
  n25 = Power(Base=Sine(None=ComponentMask(None=@n24) {g=true}) {period=2})
  Multiply(A=@n13.A, B=Multiply(A=Multiply(A=@n23, B=@n25)))
 ]
}
